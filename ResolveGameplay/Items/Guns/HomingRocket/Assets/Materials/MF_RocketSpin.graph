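# Material function that spins a mesh, by its path the homing rocket, around its object pivot using World Position Offset.
# Time times a rate parameter drives RotateAboutAxis around a local-space axis parameter, and the result adds to the incoming offset.
# FixRotateAboutAxisNormals supplies the rotated normal, replacing the incoming Normal.
# The UseRocketSpinWPO static switch selects either these modified material attributes or the untouched input.
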MaterialFunction /ResolveGameplay/Items/Guns/HomingRocket/Assets/Materials/MF_RocketSpin
graph {nodes [
  n1 = GetMaterialAttributes(None=FunctionInput() {inputName=Material, inputType=FunctionInput_MaterialAttributes}) {attributeGetTypes [#1, #2]}
  n2 = Transform(None=VectorParameter() {group=WPO_Rotation, parameterName=LocalRotationAxis, defaultValue=<rgba 1 0 0 1>}.RGB) {transformSourceType=TRANSFORMSOURCE_Local}
  n3 = Multiply(A=Time(), B=ScalarParameter() {defaultValue=-2, group=WPO_Rotation, parameterName=RotationRate})
  n4 = Add(
   A=@n1.WorldPositionOffset
   B=RotateAboutAxis(
    NormalizedRotationAxis=@n2
    RotationAngle=@n3
    PivotPoint=MaterialFunctionCall() {materialFunction=&ObjectPivotPoint}.Object Pivot Location
    Position=WorldPosition().XYZ
   )
  )
  n5 = MaterialFunctionCall(Rotation Axis=@n2, Rotation Angle=@n3) {materialFunction=&FixRotateAboutAxisNormals}
  FunctionOutput(None=StaticSwitchParameter(
    True=SetMaterialAttributes!(MaterialAttributes=@n1.MaterialAttributes, WorldPositionOffset=@n4, Normal=@n5.New Tangent Space Vertex Normal) {attributeSetTypes [#1, #2]}
    False=@n1.MaterialAttributes
   ) {parameterName=UseRocketSpinWPO})
 ]}
