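# Master material for the homing rocket item, a Default Lit material built on material attributes.
# It chains shared character functions: PreFX builds the base result, MF_RocketSpin modifies it, and PostFX takes it as its Material input.
# The PreFX AO output also feeds the PostFX AO, Distance Blend, Fresnel, Gmap and Pre Skinned Local Position inputs.
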Material /ResolveGameplay/Items/Guns/HomingRocket/Assets/Materials/M_HomingRocket_MASTER
graph {
 properties {bUseMaterialAttributes=true, pixelDepthOffsetMode=PDOM_Legacy, usedShadingModels=Default Lit}
 nodes [
  n1 = MaterialFunctionCall() {materialFunction=&BaseFN_CharacterFunction_PreFX}
  MaterialFunctionCall(
   AO=@n1.AO
   Distance Blend=@n1.AO
   Fresnel=@n1.AO
   Gmap=@n1.AO
   Pre Skinned Local Position=@n1.AO
   Material=MaterialFunctionCall!(Material=@n1.Result) {materialFunction=&/MF_RocketSpin}.Result
  ) {materialFunction=&BaseFN_CharacterFunction_PostFX}
 ]
}
