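# Post-process material for the homing rocket's scope view, tinting and desaturating the scene while highlighting lock-on targets.
# Custom stencil, depth and max-distance tests isolate targets, and a four-tap custom-depth edge filter supplies their outlines.
# MPC_WeaponScope's lock state blends target mask and colors, while its camera movement drives directional pixelation.
# Named reroutes build fisheye screen UVs, scrolling static and sweep distortion, plus grid, sonar and time-animated sphere-mask patterns.
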Material /ResolveGameplay/Items/Guns/HomingRocket/Assets/Materials/M_Weapon_HomingRocket_Postprocess
graph {
 properties {blendableLocation=BL_SceneColorAfterDOF, materialDomain=MD_PostProcess, pixelDepthOffsetMode=PDOM_Legacy, usedShadingModels=Default Lit}
 nodes [
  n1 = NamedRerouteUsage()
  n2 = Constant()
  n3 = ComponentMask(None=SceneTexture(UVs=@n1) {sceneTextureId=PPI_CustomStencil}.Color) {r=true}
  n4 = Constant() {desc=Stencil ID, r=20.5}
  n5 = Constant()
  n6 = Constant() {r=1}
  n7 = ScalarParameter() {defaultValue=25000, parameterName=Lock-on Max Dist}
  n8 = ScalarParameter() {defaultValue=250, parameterName=Lock-on Max Dist Falloff}
  n9 = SmoothStep(Min=Subtract(A=@n7, B=@n8), Max=Add(A=@n7, B=@n8), Value=NamedRerouteUsage())
  n10 = Multiply(
   A=If(A=SceneDepth(UVs=@n1).R, B=NamedRerouteUsage(), "A > B"=@n2, "A == B"=Constant() {r=1}, "A < B"=@n2) {equalsThreshold=0.1}
   B=Multiply(A=If(A=@n3, B=@n4, "A > B"=@n5, "A == B"=@n6, "A < B"=@n5) {equalsThreshold=0.51}, B=OneMinus(None=@n9))
  )
  n11 = Constant()
  n12 = If(A=@n3, B=Constant() {r=21}, "A > B"=@n11, "A == B"=Constant() {r=1}, "A < B"=@n11)
  n13 = LinearInterpolate(A=@n10, B=@n12, Alpha=CollectionParameter() {collection=/Game/Weapons/FORT_Rifles/SniperScope/MPC_WeaponScope, parameterName=Is Locked on})
  n14 = ComponentMask(None=SceneTexture(UVs=NamedRerouteUsage()) {sceneTextureId=PPI_PostProcessInput0}.Color) {b=true, g=true, r=true}
  n15 = Multiply(
   A=Desaturation(None=Power(Base=@n14) {constExponent=0.8}, Fraction=Constant() {r=1}) {luminanceFactors=<rgba 0.395 0.153 0.59 0>}
   B=VectorParameter() {parameterName=Screen Tint, defaultValue=<rgba 0.28 0.4967 0.8 0>}.RGB
  )
  n16 = Multiply(
   A=LinearInterpolate(B=LinearInterpolate(Alpha=@n13) {constA=1.6}, Alpha=NamedRerouteUsage()) {constA=1}
   B=LinearInterpolate(A=LinearInterpolate(A=@n15, B=@n14) {constAlpha=0.1}, B=@n14, Alpha=@n13)
  )
  n17 = CollectionParameter() {collection=&MPC_WeaponScope, parameterName=Is Locked on}
  n18 = LinearInterpolate(
   A=VectorParameter() {parameterName=Potential Target Color, defaultValue=<rgba 0.9 0.9 0.9 0>}.RGB
   B=VectorParameter() {parameterName=Potential Target Color Dimmed, defaultValue=<rgba 0.9 0.9 0.9 0>}.RGB
   Alpha=@n17
  )
  n19 = LinearInterpolate(
   A=VectorParameter() {parameterName=Locking-on Color, sortPriority=50, defaultValue=<rgba 1 0.65 0 1>}.RGB
   B=VectorParameter() {parameterName=Locked-on Color, sortPriority=55, defaultValue=<rgba 1 0.01 0.01 1>}.RGB
   Alpha=@n17
  )
  n20 = LinearInterpolate(Alpha=Multiply(A=NamedRerouteUsage(), B=NamedRerouteUsage())) {constA=1.5, constB=2}
  n21 = Add(A=Multiply(A=Multiply(A=@n10, B=@n20)) {constB=0.025}, B=Multiply(A=@n10, B=NamedRerouteUsage()) {constA=1})
  Add(A=@n16, B=Multiply(A=LinearInterpolate(A=@n18, B=@n19, Alpha=@n12) {constA=1}, B=@n21)) {constB=0}
  n22 = Time()
  LinearInterpolate(Alpha=ConstantBiasScale(None=Sine(None=@n22))) {constA=0.5, constB=3}
  MaterialFunctionCall(Vector 3=Subtract(A=CameraPositionWS(), B=WorldPosition().XYZ)) {materialFunction=/Engine/Functions/Engine_MaterialFunctions02/Utility/VectorLength}
  n23 = Constant()
  n24 = NamedRerouteUsage()
  n25 = Subtract(A=NamedRerouteDeclaration(None=SceneDepth(UVs=@n24).R) {name=Scene Depth, nodeColor=<rgba 1 0.2552 0.2609 1>}) {constB=100}
  n26 = LinearInterpolate(Alpha=Saturate(None=Divide(A=@n25) {constB=3000})) {constA=10, constB=100}
  n27 = Divide(A=Time() {bOverride_Period=true, period=1}, B=@n26) {constB=25}
  n28 = ViewProperty() {property=MEVP_BufferSize}
  n29 = ComponentMask(None=@n28.Property) {g=true}
  n30 = ComponentMask(None=@n28.Property) {r=true}
  n31 = Multiply(A=Subtract(A=NamedRerouteUsage(), B=AppendVector(A=@n23, B=@n27)), B=AppendVector(A=Constant() {r=1}, B=Divide(A=@n29, B=@n30)))
  n32 = Multiply(A=@n26) {constB=2}
  n33 = SmoothStep(Min=Constant() {r=0.9}, Max=Constant() {r=1}, Value=Cosine(None=Multiply(A=@n31, B=@n32) {constB=50}))
  n34 = Max(A=ComponentMask(None=@n33) {g=true}, B=ComponentMask(None=@n33) {r=true})
  Max(A=@n34, B=MaterialFunctionCall(Color=@n33) {materialFunction=&SplitComponents}.B)
  Multiply()
  Multiply(A=@n16, B=OneMinus(None=@n21) {desc=Dim outline in main pass})
  n35 = Constant() {r=6.2832}
  n36 = Constant()
  n37 = Rotator(
   Coordinate=NamedRerouteUsage()
   Time=ChannelMaskParameter(None=AppendVector(A=@n35, B=@n36)) {maskChannel=Alpha, parameterName=Glitch Direction, defaultValue=<rgba 0 0 0 1>}.R
  )
  n38 = Multiply(
   A=TextureSample(UVs=Panner(Coordinate=@n37) {speedX=6, speedY=-0.05}) {samplerType=SAMPLERTYPE_Masks, texture=/Game/Missions/Primary/RepairTheShelter/Textures/Cinematic_Screen_Static}.G
   B=TextureSample(UVs=Panner(Coordinate=@n37) {speedX=-3, speedY=-0.03}) {samplerType=SAMPLERTYPE_Masks, texture=&Cinematic_Screen_Static}.R
  )
  n39 = Multiply(A=@n38, B=ScalarParameter() {defaultValue=0.03, group=Static, parameterName=Static UV Distort})
  n40 = Constant2Vector() {r=1}
  n41 = ComponentMask(None=ScreenPosition().ViewportUV) {g=true}
  n42 = Multiply(A=Time()) {constB=-0.2}
  n43 = SmoothStep(Value=Frac(None=Add(A=@n41, B=@n42))) {constMax=1.2, constMin=0.9}
  n44 = Multiply(A=@n43, B=ScalarParameter() {defaultValue=0.01, group=Sweep Line, parameterName=Sweep UV Distort}) {constB=0.02}
  n45 = Constant2Vector() {g=1}
  NamedRerouteDeclaration(None=Add(A=Multiply(A=@n39, B=@n40), B=Multiply(A=@n44, B=@n45))) {name=Distorted UVs, nodeColor=<rgba 0 0.3203 1 1>}
  n46 = Panner(Coordinate=@n37) {speedX=5, speedY=0.1}
  n47 = Add(
   A=Multiply(A=TextureSample(UVs=@n46) {samplerType=SAMPLERTYPE_Masks, texture=&Cinematic_Screen_Static}.B) {constB=0.1}
   B=@n38
  )
  n48 = ScalarParameter() {defaultValue=6, group=Static, parameterName=Static Scale}
  n49 = ScalarParameter() {defaultValue=0.2, group=Sweep Line, parameterName=Sweep Add}
  NamedRerouteDeclaration(None=Add(A=Multiply(A=@n47, B=@n48), B=Multiply(A=@n43, B=@n49) {constB=0.2})) {name=Static, nodeColor=<rgba 0 1 0.1094 1>}
  n50 = Add(A=NamedRerouteUsage(), B=ScreenPosition().ViewportUV)
  n51 = ComponentMask(None=VectorParameter() {parameterName=TexCoordScale, defaultValue=<rgba 3 3 0 0>}.RGB) {g=true, r=true}
  n52 = VectorParameter() {parameterName=TexCoordOffset, defaultValue=<rgba -0.5 0 0 0>}
  NamedRerouteDeclaration(None=Add(A=Multiply(A=@n50, B=@n51), B=ComponentMask(None=@n52.RGB) {g=true, r=true})) {name=TexCoord, nodeColor=<rgba 1 0 0.4453 1>}
  n53 = ScreenPosition()
  n54 = Add(A=@n53.ViewportUV, B=NamedRerouteUsage())
  n55 = Constant2Vector() {g=0.5, r=0.5}
  n56 = Power(Base=Distance(A=@n53.ViewportUV, B=@n55)) {constExponent=3}
  n57 = Normalize(VectorInput=Subtract(A=@n53.ViewportUV, B=@n55))
  n58 = NamedRerouteDeclaration(None=Multiply(A=Multiply(A=@n56, B=@n57)) {constB=0.2}) {name=FishEye_Distortion, nodeColor=<rgba 0.7812 1 0 1>}
  n59 = ComponentMask(None=MaterialFunctionCall(UVs=Add(A=@n54, B=@n58), Texture Scale=Constant() {r=1}) {materialFunction=&ScaleUVsByCenter}.UVs) {g=true, r=true}
  n60 = ViewProperty() {property=MEVP_BufferSize}
  n61 = ComponentMask(None=@n60.Property) {r=true}
  n62 = ComponentMask(None=@n60.Property) {g=true}
  n63 = Multiply(A=ScalarParameter() {defaultValue=40, parameterName=Pixelation Scale}, B=AppendVector(A=Divide(A=@n61, B=@n62), B=Constant() {r=1})) {constA=40}
  n64 = Round(None=Multiply(A=@n59, B=@n63) {constB=200})
  n65 = ScalarParameter() {defaultValue=0.725, parameterName=Pixelesation Range}
  n66 = Subtract(A=TextureCoordinate()) {constB=0.5}
  n67 = Constant()
  n68 = CollectionParameter() {collection=&MPC_WeaponScope, parameterName=Camera Movement}
  n69 = AppendVector(A=ComponentMask(None=@n68) {r=true}, B=ComponentMask(None=@n68) {g=true})
  n70 = Normalize(VectorInput=@n69)
  n71 = Constant()
  n72 = Add(A=MaterialFunctionCall!(
    VectorToRotate=AppendVector(A=@n66, B=@n67)
    LookAtVector=AppendVector(A=@n70, B=@n71)
    RestingVector=Constant3Vector() {constant=<rgba 1 0 0 0>}
   ) {materialFunction=/Engine/Functions/Engine_MaterialFunctions01/Vectors/RotateVector}.Result) {constB=0.5}
  n73 = Saturate(None=ComponentMask(None=SmoothStep(Min=@n65, Value=@n72)) {r=true})
  n74 = Multiply(
   A=MaterialFunctionCall(Vector 2=@n69) {materialFunction=&VectorLength}.V2 Length
   B=ScalarParameter() {defaultValue=0.6, parameterName=Pixelation Intensity}
  )
  NamedRerouteDeclaration(None=LinearInterpolate(A=@n59, B=Divide(A=@n64, B=@n63) {constB=200}, Alpha=Multiply(A=@n73, B=@n74)) {constAlpha=1}) {name=Screen UVs, nodeColor=<rgba 1 0.7969 0 1>}
  n75 = NamedRerouteUsage()
  n76 = ComponentMask(None=SceneTexture(UVs=@n75) {sceneTextureId=PPI_CustomDepth}.Color) {r=true}
  n77 = LinearInterpolate(A=ScalarParameter() {defaultValue=16, parameterName=SampleBias_Min}, B=ScalarParameter() {defaultValue=16, parameterName=SampleBias})
  n78 = LinearInterpolate(
   A=ScalarParameter() {defaultValue=1, parameterName=MobileSampleBias_Min}
   B=ScalarParameter() {defaultValue=16, parameterName=Mobile_SampleBias}
  )
  n79 = Subtract(A=Divide(A=ViewSize(), B=Constant() {r=240}), B=Constant() {r=2}) {constB=0}
  n80 = Multiply(A=Clamp(None=Ceil(None=@n79)) {maxDefault=4, minDefault=1}, B=Multiply(A=SceneTexelSize())) {constA=2}
  n81 = Constant()
  n82 = AppendVector(A=ComponentMask(None=@n80) {desc=x, r=true}, B=@n81)
  n83 = SceneTexture(UVs=Add(A=@n75, B=@n82) {desc=x+}) {sceneTextureId=PPI_CustomDepth}
  n84 = AppendVector(A=@n81, B=ComponentMask(None=@n80) {desc=y, g=true}) {desc=y+}
  n85 = SceneTexture(UVs=Add(A=@n75, B=@n84) {desc=y+}) {sceneTextureId=PPI_CustomDepth}
  n86 = Constant() {r=-1}
  n87 = SceneTexture(UVs=Add(A=@n75, B=Multiply(A=@n82, B=@n86)) {desc=x-}) {sceneTextureId=PPI_CustomDepth}
  n88 = SceneTexture(UVs=Add(A=@n75, B=Multiply(A=@n86, B=@n84)) {desc=y-}) {sceneTextureId=PPI_CustomDepth}
  n89 = Time()
  n90 = LinearInterpolate(
   A=ScalarParameter() {parameterName=Divisor_Min}
   B=ScalarParameter() {defaultValue=-1, parameterName=Divisor_max}
   Alpha=ConstantBiasScale(None=Sine(None=@n89))
  )
  n91 = Subtract(A=PixelDepth().R, B=ScalarParameter() {defaultValue=50, group=Cartoon, parameterName=CRTNO_InnerLine_FadeStart})
  n92 = ScalarParameter() {defaultValue=500, group=Cartoon, parameterName=CRTNO_InnerLine_FadeDistance}
  n93 = LinearInterpolate(
   A=ScalarParameter() {defaultValue=-1, parameterName=MobileDivisor_Min}
   B=ScalarParameter() {defaultValue=-16, parameterName=MobileDivisor_max}
   Alpha=Clamp(None=Divide(A=@n91, B=@n92))
  )
  NamedRerouteDeclaration(None=MaterialFunctionCall!(
    Base=NamedRerouteDeclaration(None=@n76) {name=Custom Depth, nodeColor=<rgba 0 0.7891 1 1>}
    BaseSampleBias=ShadingPathSwitch(Default=@n77, Mobile=@n78)
    Scale=ScalarParameter() {defaultValue=4, parameterName=Scale}
    Sample1=ComponentMask(None=@n83.Color) {r=true}
    Sample2=ComponentMask(None=@n85.Color) {r=true}
    Sample3=ComponentMask(None=@n87.Color) {r=true}
    Sample4=ComponentMask(None=@n88.Color) {r=true}
    Divisor=ShadingPathSwitch(Default=@n90, Mobile=@n93)
    UseCenteredLines
    UseRGB
   ) {materialFunction=/Game/Packages/jl_tests/combine4Samples}.Result) {name=Outline, nodeColor=<rgba 0 1 0.2046 1>}
  n94 = TransformPosition(None=Constant3Vector() {constant=<rgba 80 -752 1920 0>}) {transformSourceType=TRANSFORMPOSSOURCE_World, transformType=TRANSFORMPOSSOURCE_TranslatedWorld}
  n95 = WorldPosition()
  n96 = Divide(A=Time())
  n97 = LinearInterpolate(Alpha=ConstantBiasScale(None=Sine(None=@n96))) {constB=500}
  NamedRerouteDeclaration(None=OneMinus(None=SphereMask(A=@n94, B=@n95.XYZ, Radius=@n97))) {name=Lock-on Progress, nodeColor=<rgba 0.8047 1 0 1>}
  n98 = Constant() {r=0.95}
  n99 = Constant() {r=0.85}
  n100 = WorldPosition()
  n101 = Saturate(None=Divide(A=ComponentMask(None=@n100.XYZ) {r=true}) {constB=1000})
  n102 = Sine(None=Divide(A=@n100.XYZ) {constB=20})
  n103 = Max(A=ComponentMask(None=@n102) {r=true}, B=ComponentMask(None=@n102) {g=true})
  n104 = ComponentMask(None=@n102) {b=true}
  NamedRerouteDeclaration(None=SmoothStep(Min=LinearInterpolate(A=@n98, B=@n99, Alpha=@n101), Max=Constant() {r=1}, Value=Max(A=@n103, B=@n104))) {name=Grid Layout, nodeColor=<rgba 1 0 0.9609 1>}
  n105 = Time()
  n106 = Divide(A=Add(A=Multiply(A=@n105) {constB=-1000}, B=SceneDepth().R)) {constB=1000}
  NamedRerouteDeclaration(None=Saturate(None=Sine(None=@n106))) {name=Sonar Pulse, nodeColor=<rgba 0 1 0.3672 1>}
  n107 = NamedRerouteUsage()
  n108 = ComponentMask(None=SceneTexture(UVs=@n107) {sceneTextureId=PPI_CustomDepth}.Color) {r=true}
  n109 = SceneDepth(UVs=@n107)
  OneMinus(None=Saturate(None=Subtract(A=@n108, B=@n109.R)))
  Saturate(None=@n34)
 ]
}
